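# TapeRecorder_Off_Cue is a Sound Cue in the Season 27 Battle Pass quest content that plays a tape recorder switch-off effect.
# A single Wave Player node feeds the TapeRecorder_Off wave straight to the output, with no randomization, modulation or mixing.
# The cue is routed to the Fort_VO_Narrator sound class, so it follows narrator voice mixing despite living in the SFX folder.
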SoundCue /BattlePassS27_Quests/Sound/SFX/TapeRecorder_Off_Cue
root=WavePlayer_1
nodes [{id=WavePlayer_1, properties {soundWaveAssetPtr=&/TapeRecorder_Off}}]
properties {maxDistance=2097152, soundClassObject=/Game/Sounds/FortSoundClassesAndModes/Fort_VO_Narrator, volumeMultiplier=1}
